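# Looping stereo MetaSound Source preset for the high-spec Akuma T2 (Nissan) vehicle engine, inheriting its graph from MSS_EngineAudio_Parent_REV_Preset.
# It feeds a three-channel Nissan REV asset and Ferrari458-folder default REV settings into that parent, alongside vehicle state inputs.
# Parameters tune throttle EQ, exhaust waveshaping, transmission, azimuth filtering and interior/exterior perspective mixes.
# Blowoff and exhaust-pop sample arrays are empty, and concurrency uses LowSpec assets despite the HighSpec name.
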MetaSoundSource /VehicleAudio/EngineAudio/Akuma/T2_Nissan/MSS_EngineAudio_Akuma_T2_HighSpec_REV_Preset
preset_parent=/VehicleAudio/Patches/MSS_EngineAudio_Parent_REV_Preset
inputs [
 Acceleration {type=Float}
 Blowoff.Sound {type=WaveAsset:Array, access=Constructor}
 Blowoff.Volume {type=Float, access=Constructor}
 BonusSpeed {type=Float}
 Character.Affiliation {type=Enum:FortAudioAffiliation}
 Character.Health {type=Float}
 Character.UniqueID {type=Int32}
 Downshift {type=Trigger}
 DriftAngle {type=Float}
 Engine.Throttle.EQ.FilterQ {type=Float, access=Constructor}
 Engine.Throttle.EQ.FilterType {type=Enum:BiquadFilterType, access=Constructor}
 Engine.Throttle.EQ.OffFreq {type=Float, access=Constructor}
 Engine.Throttle.EQ.OffGain {type=Float, access=Constructor}
 Engine.Throttle.EQ.OnFreq {type=Float, access=Constructor}
 Engine.Throttle.EQ.OnGain {type=Float, access=Constructor}
 Engine.Throttle.InterpTime {type=Time, access=Constructor}
 Engine.Throttle.Volume.Off {type=Float, access=Constructor}
 Engine.Throttle.Volume.On {type=Float, access=Constructor}
 EngineBlowoff {type=Trigger}
 Exhaust.Pops.Blowoff.Volume {type=Float, access=Constructor}
 Exhaust.Pops.Long {type=WaveAsset:Array, access=Constructor}
 Exhaust.Pops.Pitch {type=Float, access=Constructor}
 Exhaust.Pops.Shift.Volume {type=Float, access=Constructor}
 Exhaust.Pops.Short {type=WaveAsset:Array, access=Constructor}
 Exhaust.Throttle.EQ.FilterQ {type=Float, access=Constructor}
 Exhaust.Throttle.EQ.FilterType {type=Enum:BiquadFilterType, access=Constructor}
 Exhaust.Throttle.EQ.OffFreq {type=Float, access=Constructor}
 Exhaust.Throttle.EQ.OffGain {type=Float, access=Constructor}
 Exhaust.Throttle.EQ.OnFreq {type=Float, access=Constructor}
 Exhaust.Throttle.EQ.OnGain {type=Float, access=Constructor}
 Exhaust.Throttle.InterpTime {type=Time, access=Constructor}
 Exhaust.Throttle.Volume.Off {type=Float, access=Constructor}
 Exhaust.Throttle.Volume.On {type=Float, access=Constructor}
 Exhaust.Throttle.Waveshaper.Amount {type=Float, access=Constructor}
 Exhaust.Throttle.Waveshaper.AmountOffScale {type=Float, access=Constructor}
 Exhaust.Throttle.Waveshaper.AmountOnScale {type=Float, access=Constructor}
 Exhaust.Throttle.Waveshaper.Bias {type=Float, access=Constructor}
 Exhaust.Throttle.Waveshaper.Bypass {type=Bool, access=Constructor}
 Exhaust.Throttle.Waveshaper.Crossover {type=Float, access=Constructor}
 Exhaust.Throttle.Waveshaper.Mix {type=Float, access=Constructor}
 Exhaust.Throttle.Waveshaper.OutputGain {type=Float, access=Constructor}
 Exhaust.Throttle.Waveshaper.Type {type=Enum:WaveShaperType, access=Constructor}
 Fort.Vehicle.Acceleration {type=Float}
 Fort.Vehicle.BlowoffStrength {type=Float}
 Fort.Vehicle.BonusSpeed {type=Float}
 Fort.Vehicle.EnableShifting {type=Bool}
 Fort.Vehicle.Gear {type=Int32}
 Fort.Vehicle.IsGrounded {type=Bool}
 Fort.Vehicle.Load {type=Float}
 Fort.Vehicle.LowFuel {type=Bool}
 Fort.Vehicle.NormalSpeed {type=Float}
 Fort.Vehicle.OnBlowoff {type=Trigger}
 Fort.Vehicle.OnDownshift {type=Trigger}
 Fort.Vehicle.OnThrottleEngaged {type=Trigger}
 Fort.Vehicle.OnThrottleReleased {type=Trigger}
 Fort.Vehicle.OnUpshift {type=Trigger}
 Fort.Vehicle.Pitch {type=Float}
 Fort.Vehicle.RPM {type=Float}
 Fort.Vehicle.Throttle {type=Float}
 GearChangeTime {type=Time, access=Constructor}
 Interior.EQ.CutoffFrequency {type=Float, access=Constructor}
 Interior.EQ.FilterType {type=Enum:BiquadFilterType, access=Constructor}
 Interior.StereoEffect.Crossover {type=Float, access=Constructor}
 Interior.StereoEffect.Mix {type=Float, access=Constructor}
 IsLocal {type=Bool}
 Mix.Azimuth.Engine.FilterType {type=Enum:BiquadFilterType, access=Constructor}
 Mix.Azimuth.Engine.Freq.Max {type=Float, access=Constructor}
 Mix.Azimuth.Engine.Freq.Min {type=Float, access=Constructor}
 Mix.Azimuth.Engine.MinVolume {type=Float, access=Constructor}
 Mix.Azimuth.Exhaust.FilterType {type=Enum:BiquadFilterType, access=Constructor}
 Mix.Azimuth.Exhaust.Freq.Max {type=Float, access=Constructor}
 Mix.Azimuth.Exhaust.Freq.Min {type=Float, access=Constructor}
 Mix.Azimuth.Exhaust.MinVolume {type=Float, access=Constructor}
 Mix.EQ1.FilterType {type=Enum:BiquadFilterType, access=Constructor}
 Mix.EQ1.Freq {type=Float, access=Constructor}
 Mix.EQ1.Gain {type=Float, access=Constructor}
 Mix.EQ1.Q {type=Float, access=Constructor}
 Mix.EQ2.FilterType {type=Enum:BiquadFilterType, access=Constructor}
 Mix.EQ2.Freq {type=Float, access=Constructor}
 Mix.EQ2.Gain {type=Float, access=Constructor}
 Mix.EQ2.Q {type=Float, access=Constructor}
 Mix.EQ3.FilterType {type=Enum:BiquadFilterType, access=Constructor}
 Mix.EQ3.Freq {type=Float, access=Constructor}
 Mix.EQ3.Gain {type=Float, access=Constructor}
 Mix.EQ3.Q {type=Float, access=Constructor}
 Mix.EQ4.FilterType {type=Enum:BiquadFilterType, access=Constructor}
 Mix.EQ4.Freq {type=Float, access=Constructor}
 Mix.EQ4.Gain {type=Float, access=Constructor}
 Mix.EQ4.Q {type=Float, access=Constructor}
 Mix.MixPerspective {type=Int32}
 Mix.Mute.Engine {type=Bool, access=Constructor}
 Mix.Mute.Exhaust {type=Bool, access=Constructor}
 Mix.Mute.Interior {type=Bool, access=Constructor}
 Mix.Perspective.Exterior.Engine.EQ {type=Float:Array, access=Constructor}
 Mix.Perspective.Exterior.Exhaust.EQ {type=Float:Array, access=Constructor}
 Mix.Perspective.Exterior.Volume {type=Float:Array, access=Constructor}
 Mix.Perspective.Interior.Engine.EQ {type=Float:Array, access=Constructor}
 Mix.Perspective.Interior.Exhaust.EQ {type=Float:Array, access=Constructor}
 Mix.Perspective.Interior.Volume {type=Float:Array, access=Constructor}
 Multichannel {type=REV Asset, access=Constructor}
 NormalizedSpeed {type=Float}
 Pitch {type=Float}
 Quality {type=Bool}
 REV.EnableShifting {type=Bool}
 REV.Gear {type=Int32}
 REV.RPM {type=Float}
 REV.Throttle {type=Float}
 REV.Velocity {type=Float}
 REVVolume {type=Float}
 Settings {type=REV Settings, access=Constructor}
 Transmission.Enabled {type=Bool}
 Transmission.Gear.Volume {type=Float:Array, access=Constructor}
 Transmission.MaxFreq {type=Float, access=Constructor}
 Transmission.MinFreq {type=Float, access=Constructor}
 Transmission.Throttle.Volume.Off {type=Float, access=Constructor}
 Transmission.Throttle.Volume.On {type=Float, access=Constructor}
 Transmission.Volume.Overall {type=Float, access=Constructor}
 UE.Source.OnPlay {type=Trigger}
 UE.Source.Orientation.Azimuth {type=Float}
 UE.Source.Orientation.Elevation {type=Float}
 Upshift {type=Trigger}
]
interfaces [Fort.Character:1.0, Fort.Vehicle:1.0, REV:1.0, UE.OutputFormat.Stereo:1.0, UE.Source.Orientation:1.0, UE.Source:1.1]
outputs [
 "UE.OutputFormat.Stereo.Audio:0" {type=Audio}
 "UE.OutputFormat.Stereo.Audio:1" {type=Audio}
]
pages [Default {input_defaults {
   Acceleration=0
   Blowoff.Volume=0
   BonusSpeed=0
   Character.Affiliation=0
   Character.Health=0
   Character.UniqueID=0
   Downshift=null
   DriftAngle=0
   Engine.Throttle.EQ.FilterQ=1
   Engine.Throttle.EQ.FilterType=4
   Engine.Throttle.EQ.OffFreq=20
   Engine.Throttle.EQ.OffGain=0
   Engine.Throttle.EQ.OnFreq=20
   Engine.Throttle.EQ.OnGain=0
   Engine.Throttle.InterpTime=0.1
   Engine.Throttle.Volume.Off=0
   Engine.Throttle.Volume.On=0
   EngineBlowoff=null
   Exhaust.Pops.Blowoff.Volume=0
   Exhaust.Pops.Pitch=0
   Exhaust.Pops.Shift.Volume=0
   Exhaust.Throttle.EQ.FilterQ=1
   Exhaust.Throttle.EQ.FilterType=4
   Exhaust.Throttle.EQ.OffFreq=20
   Exhaust.Throttle.EQ.OffGain=0
   Exhaust.Throttle.EQ.OnFreq=20
   Exhaust.Throttle.EQ.OnGain=0
   Exhaust.Throttle.InterpTime=0.1
   Exhaust.Throttle.Volume.Off=0
   Exhaust.Throttle.Volume.On=0
   Exhaust.Throttle.Waveshaper.Amount=1
   Exhaust.Throttle.Waveshaper.AmountOffScale=0.5
   Exhaust.Throttle.Waveshaper.AmountOnScale=1
   Exhaust.Throttle.Waveshaper.Bias=0
   Exhaust.Throttle.Waveshaper.Bypass="<MetaSoundBuilderResult.SUCCEEDED: 0>"
   Exhaust.Throttle.Waveshaper.Crossover=272
   Exhaust.Throttle.Waveshaper.Mix=1
   Exhaust.Throttle.Waveshaper.OutputGain=1
   Exhaust.Throttle.Waveshaper.Type=4
   Fort.Vehicle.Acceleration=0
   Fort.Vehicle.BlowoffStrength=0
   Fort.Vehicle.BonusSpeed=0
   Fort.Vehicle.EnableShifting=null
   Fort.Vehicle.Gear=0
   Fort.Vehicle.IsGrounded="<MetaSoundBuilderResult.SUCCEEDED: 0>"
   Fort.Vehicle.Load=0
   Fort.Vehicle.LowFuel=null
   Fort.Vehicle.NormalSpeed=0
   Fort.Vehicle.OnBlowoff=null
   Fort.Vehicle.OnDownshift=null
   Fort.Vehicle.OnThrottleEngaged=null
   Fort.Vehicle.OnThrottleReleased=null
   Fort.Vehicle.OnUpshift=null
   Fort.Vehicle.Pitch=0
   Fort.Vehicle.RPM=0
   Fort.Vehicle.Throttle=0
   GearChangeTime=0.1
   Interior.EQ.CutoffFrequency=5500
   Interior.EQ.FilterType=0
   Interior.StereoEffect.Crossover=494
   Interior.StereoEffect.Mix=0.2
   IsLocal=null
   Mix.Azimuth.Engine.FilterType=0
   Mix.Azimuth.Engine.Freq.Max=20000
   Mix.Azimuth.Engine.Freq.Min=5500
   Mix.Azimuth.Engine.MinVolume=0
   Mix.Azimuth.Exhaust.FilterType=1
   Mix.Azimuth.Exhaust.Freq.Max=20
   Mix.Azimuth.Exhaust.Freq.Min=120
   Mix.Azimuth.Exhaust.MinVolume=-3
   Mix.EQ1.FilterType=4
   Mix.EQ1.Freq=100
   Mix.EQ1.Gain=0
   Mix.EQ1.Q=1
   Mix.EQ2.FilterType=4
   Mix.EQ2.Freq=300
   Mix.EQ2.Gain=0
   Mix.EQ2.Q=1
   Mix.EQ3.FilterType=4
   Mix.EQ3.Freq=5000
   Mix.EQ3.Gain=-5
   Mix.EQ3.Q=1
   Mix.EQ4.FilterType=4
   Mix.EQ4.Freq=7657.6494
   Mix.EQ4.Gain=0
   Mix.EQ4.Q=1
   Mix.MixPerspective=0
   Mix.Mute.Engine=null
   Mix.Mute.Exhaust=null
   Mix.Mute.Interior=null
   Multichannel=&/Akuma_Nissan_3CH_01
   NormalizedSpeed=0
   Pitch=0
   Quality=null
   REV.EnableShifting=null
   REV.Gear=1
   REV.RPM=0
   REV.Throttle=0
   REV.Velocity=0
   REVVolume=0
   Settings=/VehicleAudio/EngineAudio/Ferrari458/REVSetting_Default
   Transmission.Enabled=null
   Transmission.MaxFreq=550
   Transmission.MinFreq=200
   Transmission.Throttle.Volume.Off=0
   Transmission.Throttle.Volume.On=0
   Transmission.Volume.Overall=3
   UE.Source.OnPlay=null
   UE.Source.Orientation.Azimuth=0
   UE.Source.Orientation.Elevation=0
   Upshift=null
   Blowoff.Sound []
   Exhaust.Pops.Long []
   Exhaust.Pops.Short []
   Mix.Perspective.Exterior.Engine.EQ [20,1,0,1]
   Mix.Perspective.Exterior.Exhaust.EQ [20,1,0,1]
   Mix.Perspective.Exterior.Volume [-4,3,-96]
   Mix.Perspective.Interior.Engine.EQ [20,1,0,1]
   Mix.Perspective.Interior.Exhaust.EQ [20,1,0,1]
   Mix.Perspective.Interior.Volume [-96,-96,0]
   Transmission.Gear.Volume [0,0,-3,-4,-8]
  }}]
properties {
 assetImportData=&self:AssetImportData
 attenuationSettings=&Vehicle_Engine_Att
 audioPropertiesSheet=&APS_Cosmetic_Vehicle_Engine
 bLooping=true
 duration=10000
 maxDistance=12000
 numChannels=2
 outputFormat=Stereo
 priority=80
 qualitySetting=Vehicles_REV_3CH
 sampleRate=48000
 soundClassObject=&Fort_Vehicles_Engine
 soundSubmixObject=&Fortnite_SFX
 volume=1.1885
 concurrencySet [&Conc_Veh_EngineAudio_LowSpec, &Conc_Veh_EngineAudio_LowSpec_Actor]
 modulationSettings {
  highpassRouting=Union
  lowpassRouting=Union
  pitchRouting=Union
  volumeRouting=Union
  highpassModulationDestination {bEnableModulation=true, modulators [&PP_FortVehicles_Engine_HPF]}
  lowpassModulationDestination {bEnableModulation=true, modulators [&Patch_FortVehicles_Engine_LPF]}
  pitchModulationDestination {bEnableModulation=true, modulators [&DM_Bus_Vehicle_Engine_Pitch]}
  volumeModulationDestination {
   bEnableModulation=true
   modulators [&DM_VehicleEnginePatch, &Patch_FortVehicles_Engine_Vol]
  }
 }
}
